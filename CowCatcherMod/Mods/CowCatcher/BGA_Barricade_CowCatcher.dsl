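; Deployable CowCatcher barricade building that snaps to the surface below it and clears overlapping structures.
; The server validates overlaps and traces down, then damages overlapping buildings and links to the surface, or refunds a pickup on failure (optionally against armored traps).
; Clients assume success, playing a deploy animation via custom primitive data and quantized damage crinkle.
; RepNotifies apply the snap transform and hide it on death; losing the linked surface destroys it.
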(event EventBeginPlay
  (bind (_overlapped _success) (Spawning|GetAllOverlapping))
  ;[0 1 0.1341] BeginPlay
  (Networking|CanExecuteCosmeticEvents
    (:True
      ;;[0.932 0 1] Initialize Client. The client assumes that this barricade will be active. The server is in charge of tearing down the barricade if it can't be active.
      (Health|BindEventtoOnBuildingHealthChanged (EventDispatchers|CreateEvent :!event Custom|HealthChangedOnClient))
      (CallFunction|HideDeployMaterialAnimation)
      (bind _returnvalue (Utilities|Time|SetTimerbyFunctionName 0 "ActivateDeployMaterialAnimation" 0.25))
      (Variables|Default|SetActivateDeploySafetyTimer _returnvalue))
    (:False
      ;;[0 1 0.3892] The server checks if this building should remain active.
      ;; If the building should remain active the building is initialized
      ;; If not, the building is destroyed and a pick-up is spawned.
      (if _success
        (bind _returnedhit (Spawning|SnaptoSurface (Variables|Default|GetDownwardTraceLength) _overlapped))
        (if _returnedhit
          (Spawning|DestroyAllOverlapping)
          (LinkToActorComponent|PerformLinkingCheck (Variables|Default|GetFortLinkToActor) "Down" (Variables|Dynamic|GetLinkedSurfaceActor))
          (Ability|ExecuteGameplayCueReplicated (Variables|Hooks|GetGC_OnSpawn))
          (CallFunction|SetBuildingRule)
          (CallFunction|BeginCheckingforTouchingPawns) ; yellow state - only enable collision after pawns stop overlap
          (else
            _))
        (else
          _))
      ;;[1 0.4635 0.4635] This buiding is not valid, destroy it and spawn a pick-up in its place.
      (CallFunction|DestroyandSpawnPickup))))

(event Custom|SetBuildingRule
  (bind _hf_canblockbuilding (Variables|Hotfixable|GetHF_CanBlockBuilding?))
  (bind _returnvalue (ScalableFloat|GetValueasBool _hf_canblockbuilding))
  ;[0.08601 0.3473 1] Set whether players can build through CC
  (Variables|Building|SetbForceBlockBuildings _returnvalue)
  (Variables|Building|SetbDoNotBlockBuildings (not _returnvalue)))

(event OnLinkedActorDestroyed(FortLinkToActor)
  (params (DamageCauser :type Actor%20Object%20Reference))
  (Actor|DestroyActor)
  (Variables|Dynamic|SetExecuteGC? true)
  (Variables|Dynamic|SetOnDeath? true))

(event Custom|DestroyAndSpawnPickup
  (Inventory|SpawnPickupInWorld "/CowCatcherMod/Mods/CowCatcher/ID_ValetMod_CowCatcher.ID_ValetMod_CowCatcher" 1 (Transformation|GetActorLocation) (Transformation|GetActorForwardVector) 0 true true false -1 "Tossed" "TossedByPlayer")
  (Variables|Dynamic|SetExecuteGC? false)
  (Actor|DestroyActor)
  (Variables|Dynamic|SetOnDeath? true))

(event EventOnDeathServer
  (params (Damage :type Float%20%28single-precision%29) (DamageTags :type Gameplay%20Tag%20Container%20Structure%20%28by%20ref%29) (Momentum :type Vector :default "0, 0, 0") (HitInfo :type Hit%20Result%20Structure%20%28by%20ref%29) (InstigatedBy :type Controller%20Object%20Reference) (DamageCauser :type Actor%20Object%20Reference) (EffectContext :type Gameplay%20Effect%20Context%20Handle%20Structure))
  (Variables|Dynamic|SetOnDeath? true)
  (if (Variables|Dynamic|GetExecuteGC?)
    (Ability|ExecuteGameplayCueReplicated (Variables|Hooks|GetGC_OnDeath))))

(event Custom|HealthChangedOnClient
  (CallFunction|UpdateDamageMaterials))

(event EventOnPlacementBlockedByPawnChanged
  (params (bBlockedByPawn :type Boolean)))

(fn ActivateDeployMaterialAnimation
  (Rendering|Material|SetCustomPrimitiveDataFloat (Variables|Default|GetCowCatcherSM) 0 (Utilities|Time|GetGameTimeinSeconds)))

(fn CanDestroyBelow?
  (params (TestActor :type Building%20Actor%20Object%20Reference) (Normal :type Vector))
  (bind (_returnvalue_x _returnvalue_y _returnvalue_z) (Math|Vector|BreakVector (Transformation|GetActorLocation)))
  (bind _buildingtype (Variables|/Script/FortniteGame.BuildingActor|GetBuildingType TestActor))
  (if (Utilities|Enum|Equal(Enum) _buildingtype "Stairs")
    (bind _outhit (Collision|LineTraceByChannel :Start (Transformation|GetActorLocation) :End (Math|Vector|MakeVector _returnvalue_x _returnvalue_y (+ _returnvalue_z 256.0)) :TraceChannel "TraceTypeQuery1" :bTraceComplex false :DrawDebugType "None" :bIgnoreSelf true :TraceColor "(R=1.000000,G=0.000000,B=0.000000,A=1.000000)" :TraceHitColor "(R=0.000000,G=1.000000,B=0.000000,A=1.000000)" :DrawTime 5.0))
    (bind (_outhit_bblockinghit _outhit_binitialoverlap _outhit_time _outhit_distance _outhit_location _outhit_impactpoint _outhit_normal _outhit_impactnormal _outhit_physmat _outhit_hitactor) (Collision|BreakHitResult _outhit))
    (return (== _outhit_hitactor TestActor))
    (else
      (return (not (and (< (.z (Transformation|GetActorLocation TestActor)) (.z (Transformation|GetActorLocation))) (Utilities|Select :Index _buildingtype :Wall false :Floor true :Corner false :Deco false :Prop false :Stairs true :Roof true :Pillar false :SpawnedItem false :Container false :Trap false :GenericCenterCellActor false :StructuralCorner false :RoofSpacer false :None false)))))))

(fn CanSpawnOnActor?
  (params (TestActor :type Actor%20Object%20Reference))
  (Utilities|FlowControl|Sequence
    (:then_0
      (bind _asbuilding_actor (Utilities|Casting|CastToBuildingActor TestActor)
        (:then
          (LocalVariables|SetLocalBool (Utilities|Select :Index (Variables|/Script/FortniteGame.BuildingActor|GetBuildingType _asbuilding_actor) :Wall false :Floor true :Corner false :Deco false :Prop false :Stairs true :Roof true :Pillar false :SpawnedItem false :Container false :Trap false :GenericCenterCellActor false :StructuralCorner false :RoofSpacer false :None false)))
        (:CastFailed
          (Utilities|Casting|CastToPawn TestActor
            (:then
              (LocalVariables|SetLocalBool false))
            (:CastFailed
              (LocalVariables|SetLocalBool true))))))
    (:then_1
      ;[1 0.4688 0.4688] Return
      (return (LocalVariables|GetLocalBool)))))

(fn DestroyAllOverlapping
  ;[0.583 0.2655 1] Destroy any actors overlapping the barricade, except the linked surface
  (for _array_element (Variables|Dynamic|GetPendingDestructActors)
    (bind _asbuilding_actor (Utilities|Casting|CastToBuildingActor _array_element))
    (Environment|ApplyGlobalEnvironmentDamagetoActor _asbuilding_actor (ScalableFloat|GetValueatLevel (Variables|Hotfixable|GetHF_EnviroDamage))))
  (return))

(fn FilterActorstoIgnore
  (params (HitActor :type Actor%20Object%20Reference))
  ;[0.5843 0.2667 1] Ignore volumes that block the trace to the surface.
  (bind _asfort_volume (Utilities|Casting|CastToFortVolume HitActor)
    (:then
      (bind _array_element (Utilities|Array|ForEachLoopwithBreak (GameplayTags|BreakGameplayTagContainer (Variables|/Script/PlayspaceSystem.GameplayVolume|GetVolumeTags _asfort_volume)))
        (:LoopBody
          (if (!= (Utilities|Array|FindItem (LocalVariables|GetTagsToIgnore) _array_element) -1)
            (LocalVariables|SetbHasValidTags false)))
        (:Completed
          _)))
    (:CastFailed
      _))
  (return (LocalVariables|GetbHasValidTags)))

(fn GetAffiliationsToShowFor
  (return (Utilities|Array|MakeArray "Friendly" "Hostile" "Neutral")))

(fn GetAllOverlapping
  ;[0.583 0.2655 1] Get all surrounding actors that are pending destruction
  (Utilities|Array|Clear (Variables|Dynamic|GetPendingDestructActors))
  (bind _outarray (LinkToActorComponent|GetActorsLinkedTo (Variables|Default|GetFortLinkToActor)))
  (bind _outhits (Collision|MultiBoxTraceByChannel (Transformation|GetWorldLocation (Variables|Default|GetDestructionCollision)) (Transformation|GetWorldLocation (Variables|Default|GetDestructionCollision)) (Variables|/Script/Engine.BoxComponent|GetBoxExtent (Variables|Default|GetDestructionCollision)) (Transformation|GetActorRotation) "TraceTypeQuery1" false _outarray "None" true "(R=1.000000,G=0.000000,B=0.000000,A=1.000000)" "(R=0.000000,G=1.000000,B=0.000000,A=1.000000)" 1e+01))
  (bind _array_element (Utilities|Array|ForEachLoopwithBreak _outhits)
    (:LoopBody
      (bind _asbuilding_actor (Utilities|Casting|CastToBuildingActor (.HitActor (Collision|BreakHitResult _array_element)))
        (:then
          _)
        (:CastFailed
          (bind _returnvalue (GameplayTags|HasAnyMatchingGameplayTags(Message) (.HitActor (Collision|BreakHitResult _array_element)) (GameplayTags|MakeLiteralGameplayTagContainer "(GameplayTags=((TagName=\"Athena.Vehicle.Tank\"),(TagName=\"Athena.Vehicle.Valet\")))")))
          (if _returnvalue
            (Utilities|FlowControl|Sequence
              (:then_0
                _)
              (:then_1
                _))
            (return (LocalVariables|GetOverlappedActors) false))))
      (if (Spawning|CanDestroyBelow? _asbuilding_actor (.Location (Collision|BreakHitResult _array_element)))
        (Utilities|Array|Add (LocalVariables|GetOverlappedActors) _asbuilding_actor)
        (if (GameplayTags|HasMatchingGameplayTag _asbuilding_actor "(TagName=\"Item.Trap.Armored\")")
          (LocalVariables|SetOverlappedArmorWall? true))))
    (:Completed
      _))
  (bind _output_get (Variables|Dynamic|SetPendingDestructActors (LocalVariables|GetOverlappedActors)))
  (return _output_get (Math|Boolean|NANDBoolean (LocalVariables|GetOverlappedArmorWall?) (ScalableFloat|GetValueasBool (Variables|Hotfixable|GetHF_SpawnPickupVSArmor)))))

(fn HideDeployMaterialAnimation
  (Rendering|Material|SetCustomPrimitiveDataFloat (Variables|Default|GetCowCatcherSM) 0 999999995904.0))

(fn OnRep_OnDeath?
  (if (Variables|Dynamic|GetOnDeath?)
    (Development|SetHiddeninGame (Variables|Default|GetCowCatcherSM) true false)
    (Collision|SetActorEnableCollision false)
    (else
      _))
  (return))

(fn OnRep_OnSpawn?
  (if (Variables|Spawning|GetOnSpawn?)
    (Transformation|SetActorTransform (Variables|Spawning|GetRepSnapTransform))
    (Variables|Spawning|SetOnSpawn? false)
    (return)))

(fn ShouldShowSoundIndicator
  (params (PlayerController :type Fort%20Player%20Controller%20Object%20Reference))
  (bind _returnvalue (Game|GetInstigatorController))
  (Utilities|IsValid _returnvalue
    (:Is%20Valid
      (return (not (== PlayerController _returnvalue))))
    (:Is%20Not%20Valid
      (return true))))

(fn SnapToSurface
  (params (TraceLength :type Float%20%28double-precision%29) (IgnoredActors :type Array%20of%20Actor%20Object%20References))
  (bind (_returnvalue_x _returnvalue_y) (Math|Vector|BreakVector (Transformation|GetActorLocation)))
  ;[0.5843 0.2667 1] Set the location and rotation of this actor based on the surface below it
  (Utilities|Array|Add IgnoredActors (Variables|Actor|GetInstigator))
  (bind _outhits (Collision|MultiLineTraceByChannel (Transformation|GetActorLocation) (Math|Vector|MakeVector (.x (Transformation|GetActorLocation)) (.y (Transformation|GetActorLocation)) (- (.z (Transformation|GetActorLocation)) TraceLength)) "TraceTypeQuery1" false IgnoredActors "ForDuration" true "(R=1.000000,G=0.000000,B=0.000000,A=1.000000)" "(R=0.000000,G=1.000000,B=0.000000,A=1.000000)" 3e+01))
  (for _array_element _outhits
    (bind (_bblockinghit _binitialoverlap _time _distance _location _impactpoint _normal _impactnormal _physmat _hitactor) (Collision|BreakHitResult _array_element))
    (bind (_impactpoint_x _impactpoint_y _impactpoint_z) (Math|Vector|BreakVector _impactpoint))
    (bind _validactor (CallFunction|FilterActorsToIgnore _hitactor))
    (if (and _validactor (Spawning|CanSpawnonActor? _hitactor))
      (bind (_returnvalue_roll _returnvalue_pitch) (Math|Rotator|BreakRotator (Math|Rotator|MakeRotfromZX _normal (Math|Rotator|BreakRotIntoAxes (Transformation|GetActorRotation)))))
      (Variables|Spawning|SetRepSnapTransform (Math|Transform|MakeTransform (Math|Vector|MakeVector _returnvalue_x _returnvalue_y (+ _impactpoint_z (Variables|Default|GetSpawnZOffset))) (Math|Rotator|MakeRotator _returnvalue_roll _returnvalue_pitch (.yaw (Transformation|GetActorRotation))) (Transformation|GetActorScale3D)))
      (Variables|Spawning|SetOnSpawn? true)
      ;;[1 0.4635 0.4635] Return, found a valid surface to snap to.
      (return true)))
  ;[1 0.4635 0.4635] Return, failed to find snap surface
  (return false))

(fn UpdateDamageMaterials
  ;[1 1 1] Quantize
  (Networking|CanExecuteCosmeticEvents
    (:True
      (Rendering|Material|SetScalarParameterValueonMaterials (Variables|Default|GetCowCatcherSM) "Damage" (Math|Float|MapRangeClamped (/ (Math|Float|Round (* (Building|GetBuildingHealthPercentage self) (Variables|Default|GetHealthMaterialQuantizeSteps))) (Variables|Default|GetHealthMaterialQuantizeSteps)) (Variables|Default|GetHealth%25ToShowDmgCrinkle) 0.0 0.0 (Variables|Default|GetMaxCrinkleSeverity))))))

(fn ConstructionScript
  (|CallParentFunction))
