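; Cow catcher vehicle mod actor that attaches to its owning vehicle, with applied and removed FX and sound.
; Vehicle health gains can raise its health, while vehicle death or explosion removes the mod after a short delay on the server.
; Its own health drives a quantized "Damage" crinkle material, and full health can replay the applied effects.
; Vehicle hit tracking has empty branches, and Creative orphan cleanup is never called here.
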(event EventEndPlay
  (params (EndPlayReason :type EEndPlayReason%20Enum :default "Destroyed"))
  (|CallParentFunction EndPlayReason)
  (switch Utilities|FlowControl|Switch|SwitchonEEndPlayReason EndPlayReason
    (:Destroyed
      (if (Variables|Default|GetRemoveModOnDestroy)
        (Networking|SwitchHasAuthority
          (:Authority
            (Vehicle|Mod|RemoveVehicleMod (Variables|Default|GetAttachedVehicle) (Variables|Default|GetCC_ModTag))))))))

(event Custom|BindVehicleHealthEvents
  (bind _attachedvehicle (Variables|Default|GetAttachedVehicle))
  (bind _returnvalue (Utilities|ToObject(Interface) _attachedvehicle))
  ;[1 1 1] Vehicle Health Bindings - Destroy CC if vehicle destroyed and
  ; didn't damage CC enough for destroy
  (Utilities|IsValid _returnvalue
    (:Is%20Valid
      (bind _asfort_athena_vehicle (Utilities|Casting|CastToFortAthenaVehicle _returnvalue))
      (Health|BindEventtoOnHandleOutOfHealthDelegate _asfort_athena_vehicle (EventDispatchers|CreateEvent :!event Custom|OnVehicleOwnerOutOfHealth))
      (Vehicle|Destruction|BindEventtoOnVehicleExplodeDelegate _asfort_athena_vehicle (EventDispatchers|CreateEvent :!event Custom|OnOwnerVehicleExploded))
      (Health|BindEventtoOnHealthChanged _asfort_athena_vehicle (EventDispatchers|CreateEvent :!event Custom|AttachVehicleHealthChange)))))

(event Custom|SetupVehicleSkeletalHitNotifications
  ;[1 1 1] Attempt to simulate CC hitting things.
  ; Problem is, we don't get hit events when CC breaks a prop
  (if (Variables|Default|GetBindVehicleSkeletalHitEvents)
    (Collision|BindEventtoOnComponentHit (Class|FortVehicleInterface|GetSkeletalMeshComponent (Variables|Default|GetAttachedVehicle)) (delegate Custom|VehicleSkeletalMeshHit))
    (Utilities|Time|SetTimerbyEvent (delegate Custom|IgnoreEarlyCollisionOver) 1.0)))

(event Custom|IgnoreEarlyCollisionOver
  ;[1 1 1] Attempt to simulate CC hitting things.
  ; Problem is, we don't get hit events when CC breaks a prop
  (Variables|Default|SetStartTrackingVehicleSkeletalMeshHits true))

(event Custom|CleanupOwnerVehicleHealthBindings
  (bind _attachedvehicle (Variables|Default|GetAttachedVehicle))
  (bind _returnvalue (Utilities|ToObject(Interface) _attachedvehicle))
  ;[1 1 1] Vehicle Health Bindings - Destroy CC if vehicle destroyed and
  ; didn't damage CC enough for destroy
  (Utilities|IsValid _returnvalue
    (:Is%20Valid
      (bind _asfort_athena_vehicle (Utilities|Casting|CastToFortAthenaVehicle _returnvalue))
      (Health|UnbindEventfromOnHandleOutOfHealthDelegate _asfort_athena_vehicle (EventDispatchers|CreateEvent :!event Custom|OnVehicleOwnerOutOfHealth))
      (Vehicle|Destruction|UnbindEventfromOnVehicleExplodeDelegate _asfort_athena_vehicle (EventDispatchers|CreateEvent :!event Custom|OnOwnerVehicleExploded)))))

(event Custom|VehicleSkeletalMeshHit
  (params (HitComponent :type Primitive%20Component%20Object%20Reference) (OtherActor :type Actor%20Object%20Reference) (OtherComp :type Primitive%20Component%20Object%20Reference) (NormalImpulse :type Vector) (Hit :type Hit%20Result%20Structure%20%28by%20ref%29))
  (bind (_bblockinghit _binitialoverlap _time _distance _location) (Collision|BreakHitResult Hit))
  ;[1 1 1] Attempt to simulate CC hitting things.
  ; Problem is, we don't get hit events when CC breaks a prop
  (if (Variables|Default|GetStartTrackingVehicleSkeletalMeshHits)
    (if (> (Math|Vector|DotProduct (- _location (Transformation|GetActorLocation)) (Transformation|GetActorForwardVector)) 0)
      _
      (else
        _))))

(event Custom|OnOwnerVehicleExploded
  (params (LastDamageInstigator :type Controller%20Object%20Reference) (Vehicle :type Actor%20Object%20Reference))
  ;[1 1 1] Vehicle Health Bindings - Destroy CC if vehicle destroyed and
  ; didn't damage CC enough for destroy
  (CallFunction|CleanupOwnerVehicleHealthBindings)
  (Utilities|FlowControl|Delay 0.2) ; If vehicle explosion doesn't kill CC\nThen go ahead and destroy it.
  (Networking|SwitchHasAuthority
    (:Authority
      (Vehicle|Mod|RemoveVehicleMod (Variables|Default|GetAttachedVehicle) (Variables|Default|GetCC_ModTag))
      (Gameplay|Die))))

(event Custom|OnVehicleOwnerOutOfHealth
  (params (DeadVehicle :type Fort%20Athena%20Vehicle%20Object%20Reference) (InTags :type Gameplay%20Tag%20Container%20Structure%20%28by%20ref%29) (EventInstigator :type Controller%20Object%20Reference) (DamageCauser :type Actor%20Object%20Reference))
  ;[1 1 1] Vehicle Health Bindings - Destroy CC if vehicle destroyed and
  ; didn't damage CC enough for destroy
  (CallFunction|CleanupOwnerVehicleHealthBindings)
  (Utilities|FlowControl|Delay 0.2) ; If vehicle explosion doesn't kill CC\nThen go ahead and destroy it.
  (Networking|SwitchHasAuthority
    (:Authority
      (Vehicle|Mod|RemoveVehicleMod (Variables|Default|GetAttachedVehicle) (Variables|Default|GetCC_ModTag))
      (Gameplay|Die))))

(event Custom|AttachVehicleHealthChange
  (bind _attachedvehicle (Variables|Default|GetAttachedVehicle))
  (bind _returnvalue (Class|FortVehicleInterface|GetHealth _attachedvehicle))
  (bind _returnvalue_2 (Class|FortVehicleInterface|GetMaxHealth _attachedvehicle))
  (bind _returnvalue_3 (/ _returnvalue _returnvalue_2))
  (if (and (> _returnvalue_3 (Health|GetHealthPercent)) (> _returnvalue (Variables|Default|GetPreviousHealth)))
    (Health|SetHealth (Math|Float|SelectFloat (Health|GetMaxHealth) (* (Health|GetMaxHealth) _returnvalue_3) (> _returnvalue_3 0.9)))
    (else
      _))
  (Variables|Default|SetPreviousHealth _returnvalue))

(event EventBeginPlay
  (bind _returnvalue (Actor|GetOwner))
  (bind _asfort_vehicle_interface (Utilities|Casting|CastToFortVehicleInterface _returnvalue))
  (bind _staticmesh (Variables|Default|GetStaticMesh))
  ;[1 1 1] EVENT -> BeginPlay
  (Utilities|FlowControl|Sequence
    (:then_0
      (Health|BindEventtoOnBuildingHealthChanged (EventDispatchers|CreateEvent :!event Custom|HealthChanged))
      (Networking|CanExecuteCosmeticEvents
        (:True
          (CallFunction|UpdateDamageMaterials))))
    (:then_1
      (Variables|Default|SetAttachedVehicle _asfort_vehicle_interface)
      (Variables|Default|SetPreviousHealth (Class|FortVehicleInterface|GetHealth _asfort_vehicle_interface))
      (bind _returnvalue_2 (Effects|SpawnFXSystemAttached (Variables|Default|GetFX_CowCatcherApplied_NS) _staticmesh))
      (Effects|Components|ParticleSystem|SetFloatParameter _returnvalue_2 "FXScale" (.z (Variables|/Script/Engine.SceneComponent|GetRelativeScale3D _staticmesh)))
      (Audio|SpawnSoundAttached (Variables|Default|GetCowCatcherAppliedSound) (Variables|Default|GetStaticMesh)))
    (:then_2
      (CallFunction|SetupVehicleSkeletalHitNotifications)
      (CallFunction|BindVehicleHealthEvents))))

(event Custom|CreativeCleanup
  ;[1 1 1] Cow catchers that were saved to creative level before we marked that vehicle CC bga's couldn't be saved in creative, have been serialized.
  ; Not sure if this ever went live. But this will clean up during edit time in creative since players cannot select a vehicles CC
  ; Basically if an orphaned CC (no owner) exists, and we're in creative, destroy it. The owner should be a vehicle. The standalone CC is a separate BGA.
  ; This this never went live in creative, I don't believe we need to actually enable this now... but since I already wrote it I am leaving it here just in case we do.
  (if (and (Fortnite|IsCreativeorPlaygroundModePlaylist) (Networking|HasAuthority))
    (if (not (Utilities|DoesObjectImplementInterface (Actor|GetOwner) "/Script/FortniteGame.FortVehicleInterface"))
      (Utilities|Time|SetTimerbyEvent (delegate Custom|HandleCreativeNullOwnerDestroy) (Math|Random|RandomFloatinRange 5.0 6.0)) ; give it some time, just in case for some rason\nthe owner wasn't set at begin play.
    )))

(event Custom|HandleCreativeNullOwnerDestroy
  ;[1 1 1] Cow catchers that were saved to creative level before we marked that vehicle CC bga's couldn't be saved in creative, have been serialized.
  ; Not sure if this ever went live. But this will clean up during edit time in creative since players cannot select a vehicles CC
  ; Basically if an orphaned CC (no owner) exists, and we're in creative, destroy it. The owner should be a vehicle. The standalone CC is a separate BGA.
  ; This this never went live in creative, I don't believe we need to actually enable this now... but since I already wrote it I am leaving it here just in case we do.
  (Utilities|IsValid (Actor|GetOwner)
    (:Is%20Valid
      (Development|PrintString "BGA_ValetMod_CowCatcher cleaning up orphaned data")
      (Vehicle|Mod|RemoveVehicleMod (Variables|Default|GetAttachedVehicle) (Variables|Default|GetCC_ModTag))
      (Gameplay|Die))))

(event Custom|ReapplyVFX/SFX
  (bind _returnvalue (Actor|GetOwner))
  (bind _asfort_vehicle_interface (Utilities|Casting|CastToFortVehicleInterface _returnvalue))
  (bind _staticmesh (Variables|Default|GetStaticMesh))
  ;[1 1 1] EVENT -> BeginPlay
  (Variables|Default|SetAttachedVehicle _asfort_vehicle_interface)
  (Variables|Default|SetPreviousHealth (Class|FortVehicleInterface|GetHealth _asfort_vehicle_interface))
  (bind _returnvalue_2 (Effects|SpawnFXSystemAttached (Variables|Default|GetFX_CowCatcherApplied_NS) _staticmesh))
  (Effects|Components|ParticleSystem|SetFloatParameter _returnvalue_2 "FXScale" (.z (Variables|/Script/Engine.SceneComponent|GetRelativeScale3D _staticmesh)))
  (Audio|SpawnSoundAttached (Variables|Default|GetCowCatcherAppliedSound) (Variables|Default|GetStaticMesh)))

(event Custom|SetAttachedVehicle
  (params (AttachedVehicle :type Fort%20Vehicle%20Interface%20Interface))
  ;[1 1 1] SetAttached Vehicle - Vehicle Ref set from projectile (Server only)
  (Variables|Default|SetAttachedVehicle AttachedVehicle))

(event EventOnDeathPlayEffects
  (params (Damage :type Float%20%28single-precision%29) (DamageTags :type Gameplay%20Tag%20Container%20Structure%20%28by%20ref%29) (Momentum :type Vector :default "0, 0, 0") (HitInfo :type Hit%20Result%20Structure%20%28by%20ref%29) (InstigatedBy :type Fort%20Pawn%20Object%20Reference) (DamageCauser :type Actor%20Object%20Reference) (EffectContext :type Gameplay%20Effect%20Context%20Handle%20Structure))
  (bind _self self)
  (bind _returnvalue (Transformation|GetActorLocation _self))
  (bind _staticmesh (Variables|Default|GetStaticMesh))
  (bind _relativescale3d (Variables|/Script/Engine.SceneComponent|GetRelativeScale3D _staticmesh))
  ;[1 1 1] Death Effects
  (bind _returnvalue_2 (Effects|Components|ParticleSystem|SpawnFXAtLocation (Variables|Default|GetFX_CowCatcherRemoved_NS) _returnvalue (Transformation|GetActorRotation) _relativescale3d))
  (Effects|Components|ParticleSystem|SetFloatParameter _returnvalue_2 "FXScale" (.z _relativescale3d))
  (Audio|SpawnSoundatLocation (Variables|Default|GetCowCatcherRemovedSound) _returnvalue)
  (Rendering|SetVisibility (Variables|Default|GetStaticMesh) false false))

(event EventOnDeathServer
  (params (Damage :type Float%20%28single-precision%29) (DamageTags :type Gameplay%20Tag%20Container%20Structure%20%28by%20ref%29) (Momentum :type Vector :default "0, 0, 0") (HitInfo :type Hit%20Result%20Structure%20%28by%20ref%29) (InstigatedBy :type Controller%20Object%20Reference) (DamageCauser :type Actor%20Object%20Reference) (EffectContext :type Gameplay%20Effect%20Context%20Handle%20Structure))
  ;[1 1 1] Death Effects
  (Gameplay|Die))

(event Custom|HealthChanged
  (bind _attachedvehicle (Variables|Default|GetAttachedVehicle))
  (bind _returnvalue (Class|FortVehicleInterface|GetHealth _attachedvehicle))
  ;[1 1 1] Health Changed
  (if (!= _returnvalue (Variables|Default|GetPreviousHealth))
    (CallFunction|UpdateDamageMaterials)
    (Variables|Default|SetPreviousHealth _returnvalue)))

(fn CalculateLaunchDirection
  ;[1 1 1] Toss vehicle behind us
  (return (Math|Vector|NegateVector (Math|Vector|GetForwardVector (Transformation|GetWorldRotation (Variables|Default|GetStaticMesh))))))

(fn CanDestroyBuilding
  (params (Actor :type Actor%20Object%20Reference))
  ;[1 1 1] Check to see if we should destroy the building actor
  (return (Math|Boolean|ANDBoolean (Variables|/Script/Engine.Actor|GetbCanBeDamaged Actor) (not (Actor|ActorHasTag Actor "NoVehicleDamage")) (Class|FortVehicleInterface|HasPermissiontoAlterBuildings (Variables|Default|GetAttachedVehicle)))))

(fn RandomizeSpinThrust
  (return (Math|Vector|MakeVector 0.0 (Math|Random|RandomFloatinRange (Math|Float|Min(Float) (* (Variables|Default|GetSpinMag) -1.0) -2e+02) (Math|Float|Max(Float) (Variables|Default|GetSpinMag) 2e+02)) (Math|Random|RandomFloatinRange (Math|Float|Min(Float) (* (Variables|Default|GetSpinMag) -1.0) -2e+02) (Math|Float|Max(Float) (Variables|Default|GetSpinMag) 2e+02)))))

(fn UpdateDamageMaterials
  (bind _self self)
  (bind _returnvalue (Building|GetBuildingHealthPercentage _self))
  (Rendering|Material|SetScalarParameterValueonMaterials (Variables|Default|GetStaticMesh) "Damage" (Math|Float|MapRangeClamped (/ (Math|Float|Round (* _returnvalue (Variables|Default|GetHealthMaterialQuantizeSteps))) (Variables|Default|GetHealthMaterialQuantizeSteps)) (Variables|Default|GetHealth%25ToShowDmgCrinkle) 0.0 0.0 (Variables|Default|GetMaxCrinkleSeverity)))
  (if (== _returnvalue 1.0)
    (bind _returnvalue_2 (GameplayEffects|GetActiveEffectswithAllTags (Ability|GetFortAbilitySystemComponent self) (Variables|Default|GetHealingTag)))
    (if (>= (Utilities|Array|Length _returnvalue_2) 1)
      (Development|PrintString "IM HEALLLLED" true true "(R=0.000000,G=0.660000,B=1.000000,A=1.000000)" 1e+01)
      (else
        (CallFunction|ReapplyVFX/SFX)
        (Development|PrintString "IM REPAIRRRED" true true "(R=0.000000,G=0.660000,B=1.000000,A=1.000000)" 1e+01)))))

(fn ConstructionScript
  (|CallParentFunction))
