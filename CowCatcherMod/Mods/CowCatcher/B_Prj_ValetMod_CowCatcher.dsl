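; Thrown Cow Catcher projectile that applies its vehicle mod on impact or spawns a barricade on other valid surfaces.
; EventHit checks vehicle health and mod eligibility, then applies impact and mod on the server, with a timer delaying the mod in replays.
; Water entry, or repeated vehicle hits, instead drops a weapon pickup.
; Hit and EnteredWater flags plus DoOnce gates keep these outcomes from overlapping.
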(event Custom|SpawnPickupMod
  ;[1 1 1] If we hit the same vehicle multiple times very quickly then spawn a pickup. This is to fix FORT-364546
  (Utilities|FlowControl|DoOnce false
    (:Completed
      (Rendering|SetActorHiddenInGame true)
      (Inventory|SpawnPickupInWorld (Variables|Default|GetCowCatcherWID) 1 (Transformation|GetWorldLocation (Variables|Collision|GetCapsuleComponent)) "0.000000,0.000000,1.000000" 0 true false)
      (Actor|DestroyActor))))

(event Custom|ApplyImpact
  ;[1 1 1] EVENT -> On Hit
  (Rendering|SetVisibility (Variables|Default|GetSM_CowCatcher) false true)
  (bind _returnvalue (PlayerInfo|GetOwnerPlayerController))
  (Audio|BroadcastAbilityImpactatLocation 0 (stale Collision|EventHit _asfort_athena_vehicle) (Math|Vector|MakeVector 0.0 0.0 (Variables|Default|GetBumpPositionZ)) 777.0 "None" false) ; Add a little bump to sell the Application
)

(event Collision|EventHit
  (params (MyComp :type Primitive%20Component%20Object%20Reference) (Other :type Actor%20Object%20Reference) (OtherComp :type Primitive%20Component%20Object%20Reference) (bSelfMoved :type Boolean) (HitLocation :type Vector :default "0, 0, 0") (HitNormal :type Vector :default "0, 0, 0") (NormalImpulse :type Vector :default "0, 0, 0") (Hit :type Hit%20Result%20Structure%20%28by%20ref%29))
  (bind (_bblockinghit _binitialoverlap _time _distance _location _impactpoint _normal _impactnormal _physmat _hitactor) (Collision|BreakHitResult Hit))
  ;[1 1 1] EVENT -> On Hit
  ;;[1 1 1] If already entered water, don't do anything with this hit. Set hit so it can be checked in water interaction.
  (if (Variables|Default|GetEnteredWater)
    (Variables|Spawning|ValidityChecks|SetbSpawnOnCollision false)
    (else
      (Variables|Default|SetHit true)
      ;;[1 1 1] Apply Mod to Vehicle
      (bind _asfort_athena_vehicle (Utilities|Casting|CastToFortAthenaVehicle _hitactor)
        (:then
          (if (Vehicle|Mod|CanApplyMod _asfort_athena_vehicle (Variables|Default|GetMod%20Tag))
            (if (CallFunction|ShouldApplyMod _asfort_athena_vehicle)
              (Utilities|FlowControl|DoOnce false
                (:Completed
                  ;;[1 1 1] Fix for impact showing late in replays
                  (if (Replay|GetIsPlayingReplay)
                    (Development|PrintString "HELLO IM IN REPLAYS" true true "(R=0.000000,G=0.660000,B=1.000000,A=1.000000)" 1e+01)
                    (CallFunction|ApplyImpact)
                    (Utilities|Time|SetTimerbyEvent (EventDispatchers|CreateEvent :!event Custom|ApplyMod) 0.3)
                    (elif (Networking|IsServer)
                      (CallFunction|ApplyImpact)
                      (CallFunction|ApplyMod)))))
              (else
                _)))
          ;[1 1 1] If we hit the same vehicle multiple times very quickly then spawn a pickup. This is to fix FORT-364546
          (if (Networking|IsServer)
            _)
          (CallFunction|ParentHit))
        (:CastFailed
          (Utilities|Casting|CastToFortPawn _hitactor
            (:then
              (bind _asfort_player_pawn (Utilities|Casting|CastToFortPlayerPawn _hitactor)
                (:then
                  (if (or (Vehicle|IsinVehicle _asfort_player_pawn) (Health|IsDBNO _asfort_player_pawn))
                    _
                    (else
                      _)))
                (:CastFailed
                  _)))
            (:CastFailed
              (if (CallFunction|IsValidHit? _hitactor)
                (Utilities|FlowControl|DoOnce false
                  (:Completed
                    (CallFunction|SpawnBarricadeBGA)))
                (else
                  _))))
          (CallFunction|ParentHit)
          (Audio|SpawnSoundatLocation "/CowCatcherMod/Sound/CowCatcher_ImpactDamage_Cue.CowCatcher_ImpactDamage_Cue" HitLocation))))))

(event Custom|ApplyMod
  ;[1 1 1] EVENT -> On Hit
  (Vehicle|Mod|ApplyVehicleMod (stale Collision|EventHit _asfort_athena_vehicle) (Variables|Default|GetMod%20Tag) (stale Custom|ApplyImpact _returnvalue))
  (Actor|DestroyActor))

(event Custom|ParentHit
  ;[1 1 1] EVENT -> On Hit
  (CallFunction|SpawnPickupMod)
  (CallFunction|DisposeProjectile)
  (Collision|SetActorEnableCollision false) ; If we spawned the pickup then disable collision so spawned pickup doesn't collide with us
)

(event EventOnStop
  (params (Hit :type Hit%20Result%20Structure%20%28by%20ref%29))
  ;[1 1 1] EVENT -> On Hit
  (|CallParentFunction Hit)
  (CallFunction|DisposeProjectile))

(event Custom|DisposeProjectile
  ;[1 1 1] EVENT -> On Stop
  (Rendering|SetVisibility (Variables|Default|GetSM_CowCatcher) false true)
  (Rendering|SetActorHiddenInGame true)
  (Rendering|SetVisibility (Variables|Default|GetParticleSystem) false false)
  (Game|Components|ProjectileMovement|StopSimulating (Variables|Movement|GetProjectileMovementComponent) (Collision|MakeHitResult false)))

(event EventBeginPlay
  ;[1 1 1] BeginPlay -> Fix throwing the projectile whilst in water, to allow it time to leave the pawns hands
  (|CallParentFunction)
  (Variables|Default|SetFiredFromSecond? (GameplayTags|HasMatchingGameplayTag (Utilities|Casting|CastToFortPlayerPawnAthena (Game|GetInstigator)) "(TagName=\"Abilities.Generic.ActiveAbility.ActivateSecondary\")"))
  (Utilities|FlowControl|Delay 0.2)
  (if (Water|IsinWater (Variables|Water|GetWaterInteractionComponent))
    (CallFunction|HasEnteredWater)))

(event Custom|HasEnteredWater
  (bind _capsulecomponent (Variables|Collision|GetCapsuleComponent))
  (bind _returnvalue (Transformation|GetWorldLocation _capsulecomponent))
  ;[1 1 1] Has Entered Water - Spawn Pickup
  ;;[1 1 1] If already entered Hit, don't do anything with the water interaction. Set Entered Water so it can be checked in Hit interaction.
  (if (not (Variables|Default|GetHit))
    (Utilities|FlowControl|DoOnce false
      (:Completed
        (Variables|Default|SetEnteredWater true)
        (Effects|Components|ParticleSystem|SpawnFXAtLocation (Variables|Default|GetWaterSplashFX) _returnvalue)
        (Rendering|SetActorHiddenInGame true)
        (Inventory|SpawnPickupInWorld (Variables|Default|GetCowCatcherWID) 1 _returnvalue "0.000000,0.000000,1.000000" 0 true false)
        (Actor|DestroyActor)))))

(event Custom|SpawnBarricadeBGA
  (bind (_returnvalue_roll _returnvalue_pitch _returnvalue_yaw) (Math|Rotator|BreakRotator (Transformation|GetActorRotation)))
  ;[0.08601 0.3473 1] Spawn Barricade BGA
  (Rendering|SetVisibility (Variables|Default|GetSM_CowCatcher) false false)
  (CallFunction|DisposeProjectile)
  (Collision|SetActorEnableCollision false) ; If we spawned the BGA then disable collision so spawned BGA doesn't collide with us
  (Networking|CanExecuteCosmeticEvents
    (:False
      (Game|SpawnBuildingActorfromClass (Variables|Default|GetBarricadeClass) (Math|Transform|MakeTransform (Transformation|GetActorLocation) (Math|Rotator|MakeRotator 0.0 0.0 _returnvalue_yaw) "1.000000,1.000000,1.000000") (Game|GetInstigator) false 0 (Game|GetInstigator)))))

(event OnEnterWaterEvent(WaterInteractionComponent)
  (params (WaterBody :type Fort%20Water%20Body%20Actor%20Object%20Reference) (WaterInteractionComponent :type Fort%20Water%20Interaction%20Component%20Object%20Reference) (bIsFirstBody :type Boolean))
  ;[1 1 1] Has Entered Water - Spawn Pickup
  (Variables|Spawning|ValidityChecks|SetbSpawnOnCollision false)
  (CallFunction|HasEnteredWater))

(fn CalculateLaunchVelocity
  (params (PlayerPawn :type Fort%20Pawn%20Object%20Reference))
  (return (Math|Vector|MakeVector (.x (* (Math|Vector|Normalize (- (Transformation|GetActorLocation PlayerPawn) (Transformation|GetActorLocation))) (ScalableFloat|GetValueatLevel (Variables|HotfixRows|GetHitPlayerLaunchMag)))) (.y (* (Math|Vector|Normalize (- (Transformation|GetActorLocation PlayerPawn) (Transformation|GetActorLocation))) (ScalableFloat|GetValueatLevel (Variables|HotfixRows|GetHitPlayerLaunchMag)))) (* (ScalableFloat|GetValueatLevel (Variables|HotfixRows|GetHitPlayerLaunchMag)) (ScalableFloat|GetValueatLevel (Variables|HotfixRows|GetHitPlayerLaunchZMulti))))))

(fn IsValidHit?
  (params (HitActor :type Actor%20Object%20Reference))
  (return (not (or (Utilities|ClassIsChildOf (Utilities|GetClass HitActor) "/Script/FortniteGame.BuildingTrap") (== (Utilities|GetClass HitActor) "/CowCatcherMod/Mods/CowCatcher/BGA_ValetMod_CowCatcher.BGA_ValetMod_CowCatcher_C")))))

(fn ShouldApplyMod
  (params (Vehicle :type Fort%20Athena%20Vehicle%20Object%20Reference))
  ;[1 1 1] If none of these are true then we should apply the mod
  (return (not (<= (Class|FortVehicleInterface|GetHealth Vehicle) 0.0))))

(fn ConstructionScript
  (|CallParentFunction))
